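# Physics asset for the face accessory part of the F_MED_Stands BLK player character, defining bodies and joints for its hair bones.
# Twelve hair chains (a to l) each link three capsule bodies in sequence, with every root constrained to a small head capsule.
# Root capsules are smaller than the two segments after them.
# All joints share the same tight swing and twist limit angles, suggesting subtle, stiff secondary motion.
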PhysicsAsset /Game/Characters/Player/Female/Medium/Bodies/F_MED_Stands/Meshes/Parts/BLK/F_MED_Stands_BLK_FaceAcc_Physics
bodies {
 hair_a_01 [{shapeType=Capsule, length=0.5, radius=0.5, shapeName=hair_a_01_capsule, rotation=<rot 90 0 0>}]
 hair_a_02 [{shapeType=Capsule, length=1.5, radius=1.5, shapeName=hair_a_02_capsule, center=<1.6815 0 0>, rotation=<rot 90 0 0>}]
 hair_a_03 [{shapeType=Capsule, length=1.5, radius=1.5, shapeName=hair_a_03_capsule, center=<0 1.081 0>, rotation=<rot 90 0 0>}]
 hair_b_01 [{shapeType=Capsule, length=0.5, radius=0.5, shapeName=hair_b_01_capsule, rotation=<rot 90 0 0>}]
 hair_b_02 [{shapeType=Capsule, length=1.5, radius=1.5, shapeName=hair_b_02_capsule, center=<1.6815 0 0>, rotation=<rot 90 0 0>}]
 hair_b_03 [{shapeType=Capsule, length=1.5, radius=1.5, shapeName=hair_b_03_capsule, center=<0 1.081 0>, rotation=<rot 90 0 0>}]
 hair_c_01 [{shapeType=Capsule, length=0.5, radius=0.5, shapeName=hair_c_01_capsule, rotation=<rot 90 0 0>}]
 hair_c_02 [{shapeType=Capsule, length=1.5, radius=1.5, shapeName=hair_c_02_capsule, center=<1.6815 0 0>, rotation=<rot 90 0 0>}]
 hair_c_03 [{shapeType=Capsule, length=1.5, radius=1.5, shapeName=hair_c_03_capsule, center=<0 1.081 0>, rotation=<rot 90 0 0>}]
 hair_d_01 [{shapeType=Capsule, length=0.5, radius=0.5, shapeName=hair_d_01_capsule, rotation=<rot 90 0 0>}]
 hair_d_02 [{shapeType=Capsule, length=1.5, radius=1.5, shapeName=hair_d_02_capsule, center=<1.7118 0 0>, rotation=<rot 90 0 0>}]
 hair_d_03 [{shapeType=Capsule, length=1.5, radius=1.5, shapeName=hair_d_03_capsule, center=<1.7118 0 0>, rotation=<rot 90 0 0>}]
 hair_e_01 [{shapeType=Capsule, length=0.5, radius=0.5, shapeName=hair_e_01_capsule, rotation=<rot 90 0 0>}]
 hair_e_02 [{shapeType=Capsule, length=1.5, radius=1.5, shapeName=hair_e_02_capsule, center=<2.0326 0 0>, rotation=<rot 90 0 0>}]
 hair_e_03 [{shapeType=Capsule, length=1.5, radius=1.5, shapeName=hair_e_03_capsule, center=<2.0326 0 0>, rotation=<rot 90 0 0>}]
 hair_f_01 [{shapeType=Capsule, length=0.5, radius=0.5, shapeName=hair_f_01_capsule, rotation=<rot 90 0 0>}]
 hair_f_02 [{shapeType=Capsule, length=1.5, radius=1.5, shapeName=hair_f_02_capsule, center=<2.3513 0 0>, rotation=<rot 90 0 0>}]
 hair_f_03 [{shapeType=Capsule, length=1.5, radius=1.5, shapeName=hair_f_03_capsule, center=<2.3513 0 0>, rotation=<rot 90 0 0>}]
 hair_g_01 [{shapeType=Capsule, length=0.5, radius=0.5, shapeName=hair_g_01_capsule, rotation=<rot 90 0 0>}]
 hair_g_02 [{shapeType=Capsule, length=1.5, radius=1.5, shapeName=hair_g_02_capsule, center=<1.7118 0 0>, rotation=<rot 90 0 0>}]
 hair_g_03 [{shapeType=Capsule, length=1.5, radius=1.5, shapeName=hair_g_03_capsule, center=<1.7118 0 0>, rotation=<rot 90 0 0>}]
 hair_h_01 [{shapeType=Capsule, length=0.5, radius=0.5, shapeName=hair_h_01_capsule, rotation=<rot 90 0 0>}]
 hair_h_02 [{shapeType=Capsule, length=1.5, radius=1.5, shapeName=hair_h_02_capsule, center=<1.7118 0 0>, rotation=<rot 90 0 0>}]
 hair_h_03 [{shapeType=Capsule, length=1.5, radius=1.5, shapeName=hair_h_03_capsule, center=<1.7118 0 0>, rotation=<rot 90 0 0>}]
 hair_i_01 [{shapeType=Capsule, length=0.5, radius=0.5, shapeName=hair_i_01_capsule, rotation=<rot 90 0 0>}]
 hair_i_02 [{shapeType=Capsule, length=1.5, radius=1.5, shapeName=hair_i_02_capsule, center=<2.3513 0 0>, rotation=<rot 90 0 0>}]
 hair_i_03 [{shapeType=Capsule, length=1.5, radius=1.5, shapeName=hair_i_03_capsule, center=<2.3513 0 0>, rotation=<rot 90 0 0>}]
 hair_j_01 [{shapeType=Capsule, length=0.5, radius=0.5, shapeName=hair_j_01_capsule, rotation=<rot 90 0 0>}]
 hair_j_02 [{shapeType=Capsule, length=1.5, radius=1.5, shapeName=hair_j_02_capsule, center=<2.3513 0 0>, rotation=<rot 90 0 0>}]
 hair_j_03 [{shapeType=Capsule, length=1.5, radius=1.5, shapeName=hair_j_03_capsule, center=<2.3513 0 0>, rotation=<rot 90 0 0>}]
 hair_k_01 [{shapeType=Capsule, length=0.5, radius=0.5, shapeName=hair_k_01_capsule, rotation=<rot 90 0 0>}]
 hair_k_02 [{shapeType=Capsule, length=1.5, radius=1.5, shapeName=hair_k_02_capsule, center=<2.0326 0 0>, rotation=<rot 90 0 0>}]
 hair_k_03 [{shapeType=Capsule, length=1.5, radius=1.5, shapeName=hair_k_03_capsule, center=<2.0326 0 0>, rotation=<rot 90 0 0>}]
 hair_l_01 [{shapeType=Capsule, length=0.5, radius=0.5, shapeName=hair_l_01_capsule, center=<2.0326 0 0>, rotation=<rot 90 0 0>}]
 hair_l_02 [{shapeType=Capsule, length=1.5, radius=1.5, shapeName=hair_l_02_capsule, rotation=<rot 90 0 0>}]
 hair_l_03 [{shapeType=Capsule, length=1.5, radius=1.5, shapeName=hair_l_03_capsule, center=<2.0326 0 0>, rotation=<rot 90 0 0>}]
 head [{shapeType=Capsule, length=0, radius=0.5, shapeName=head_capsule}]
}
constraints [
 {bone1Name=hair_l_02, bone2Name=hair_l_01, swing1LimitDegrees=5, swing2LimitDegrees=5, twistLimitDegrees=5}
 {bone1Name=hair_l_03, bone2Name=hair_l_02, swing1LimitDegrees=5, swing2LimitDegrees=5, twistLimitDegrees=5}
 {bone1Name=hair_k_02, bone2Name=hair_k_01, swing1LimitDegrees=5, swing2LimitDegrees=5, twistLimitDegrees=5}
 {bone1Name=hair_k_03, bone2Name=hair_k_02, swing1LimitDegrees=5, swing2LimitDegrees=5, twistLimitDegrees=5}
 {bone1Name=hair_j_02, bone2Name=hair_j_01, swing1LimitDegrees=5, swing2LimitDegrees=5, twistLimitDegrees=5}
 {bone1Name=hair_j_03, bone2Name=hair_j_02, swing1LimitDegrees=5, swing2LimitDegrees=5, twistLimitDegrees=5}
 {bone1Name=hair_i_02, bone2Name=hair_i_01, swing1LimitDegrees=5, swing2LimitDegrees=5, twistLimitDegrees=5}
 {bone1Name=hair_i_03, bone2Name=hair_i_02, swing1LimitDegrees=5, swing2LimitDegrees=5, twistLimitDegrees=5}
 {bone1Name=hair_h_02, bone2Name=hair_h_01, swing1LimitDegrees=5, swing2LimitDegrees=5, twistLimitDegrees=5}
 {bone1Name=hair_h_03, bone2Name=hair_h_02, swing1LimitDegrees=5, swing2LimitDegrees=5, twistLimitDegrees=5}
 {bone1Name=hair_g_02, bone2Name=hair_g_01, swing1LimitDegrees=5, swing2LimitDegrees=5, twistLimitDegrees=5}
 {bone1Name=hair_g_03, bone2Name=hair_g_02, swing1LimitDegrees=5, swing2LimitDegrees=5, twistLimitDegrees=5}
 {bone1Name=hair_f_02, bone2Name=hair_f_01, swing1LimitDegrees=5, swing2LimitDegrees=5, twistLimitDegrees=5}
 {bone1Name=hair_f_03, bone2Name=hair_f_02, swing1LimitDegrees=5, swing2LimitDegrees=5, twistLimitDegrees=5}
 {bone1Name=hair_e_02, bone2Name=hair_e_01, swing1LimitDegrees=5, swing2LimitDegrees=5, twistLimitDegrees=5}
 {bone1Name=hair_e_03, bone2Name=hair_e_02, swing1LimitDegrees=5, swing2LimitDegrees=5, twistLimitDegrees=5}
 {bone1Name=hair_d_02, bone2Name=hair_d_01, swing1LimitDegrees=5, swing2LimitDegrees=5, twistLimitDegrees=5}
 {bone1Name=hair_d_03, bone2Name=hair_d_02, swing1LimitDegrees=5, swing2LimitDegrees=5, twistLimitDegrees=5}
 {bone1Name=hair_c_02, bone2Name=hair_c_01, swing1LimitDegrees=5, swing2LimitDegrees=5, twistLimitDegrees=5}
 {bone1Name=hair_c_03, bone2Name=hair_c_02, swing1LimitDegrees=5, swing2LimitDegrees=5, twistLimitDegrees=5}
 {bone1Name=hair_b_02, bone2Name=hair_b_01, swing1LimitDegrees=5, swing2LimitDegrees=5, twistLimitDegrees=5}
 {bone1Name=hair_b_03, bone2Name=hair_b_02, swing1LimitDegrees=5, swing2LimitDegrees=5, twistLimitDegrees=5}
 {bone1Name=hair_a_02, bone2Name=hair_a_01, swing1LimitDegrees=5, swing2LimitDegrees=5, twistLimitDegrees=5}
 {bone1Name=hair_a_03, bone2Name=hair_a_02, swing1LimitDegrees=5, swing2LimitDegrees=5, twistLimitDegrees=5}
 {bone1Name=hair_l_01, bone2Name=head, swing1LimitDegrees=5, swing2LimitDegrees=5, twistLimitDegrees=5}
 {bone1Name=hair_k_01, bone2Name=head, swing1LimitDegrees=5, swing2LimitDegrees=5, twistLimitDegrees=5}
 {bone1Name=hair_j_01, bone2Name=head, swing1LimitDegrees=5, swing2LimitDegrees=5, twistLimitDegrees=5}
 {bone1Name=hair_i_01, bone2Name=head, swing1LimitDegrees=5, swing2LimitDegrees=5, twistLimitDegrees=5}
 {bone1Name=hair_h_01, bone2Name=head, swing1LimitDegrees=5, swing2LimitDegrees=5, twistLimitDegrees=5}
 {bone1Name=hair_g_01, bone2Name=head, swing1LimitDegrees=5, swing2LimitDegrees=5, twistLimitDegrees=5}
 {bone1Name=hair_f_01, bone2Name=head, swing1LimitDegrees=5, swing2LimitDegrees=5, twistLimitDegrees=5}
 {bone1Name=hair_e_01, bone2Name=head, swing1LimitDegrees=5, swing2LimitDegrees=5, twistLimitDegrees=5}
 {bone1Name=hair_d_01, bone2Name=head, swing1LimitDegrees=5, swing2LimitDegrees=5, twistLimitDegrees=5}
 {bone1Name=hair_c_01, bone2Name=head, swing1LimitDegrees=5, swing2LimitDegrees=5, twistLimitDegrees=5}
 {bone1Name=hair_b_01, bone2Name=head, swing1LimitDegrees=5, swing2LimitDegrees=5, twistLimitDegrees=5}
 {bone1Name=hair_a_01, bone2Name=head, swing1LimitDegrees=5, swing2LimitDegrees=5, twistLimitDegrees=5}
]
